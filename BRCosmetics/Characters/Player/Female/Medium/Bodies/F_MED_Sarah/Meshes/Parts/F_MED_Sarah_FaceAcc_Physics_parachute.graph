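# Physics asset for the parachute variant of Sarah's face accessory, covering her ponytail and three hair strands.
# Capsules on upper spine, neck, head and left shoulder form a locked body chain, including an oversized, offset spine_05 capsule.
# The five-bone ponytail hangs from a head-locked base, with limits generally widening toward the tip.
# Each single-bone strand has limited motion against its base, whose constraint to the head is free.
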PhysicsAsset /BRCosmetics/Characters/Player/Female/Medium/Bodies/F_MED_Sarah/Meshes/Parts/F_MED_Sarah_FaceAcc_Physics_parachute
bodies {
 base_hair_01 [{shapeType=Capsule, length=0.505, radius=0.505, shapeName=base_hair_01_capsule, rotation=<rot 0 -90 0>}]
 base_hair_02 [{shapeType=Capsule, length=0.505, radius=0.505, shapeName=base_hair_02_capsule, rotation=<rot 0 -90 0>}]
 base_hair_03 [{shapeType=Capsule, length=0.505, radius=0.505, shapeName=base_hair_03_capsule, rotation=<rot 0 -90 0>}]
 base_ponytail_1 [{shapeType=Capsule, length=0.505, radius=0.505, shapeName=base_ponytail_1_capsule, rotation=<rot 0 -90 0>}]
 clavicle_l [
  {
   shapeType=Capsule
   length=7.6907
   radius=6.6756
   shapeName=clavicle_l_capsule
   center=<3.6475 1.5718 -4.8503>
   rotation=<rot 0.5169 -94.9243 96.0901>
  }
  {
   shapeType=Capsule
   length=9.7013
   radius=8.2174
   shapeName=clavicle_l_capsule
   center=<5.1353 -1.3934 -4.034>
   rotation=<rot -0.0007 -89.9999 102.964>
  }
  {
   shapeType=Capsule
   length=10.5311
   radius=6.4858
   shapeName=clavicle_l_capsule
   center=<7.8888 -12.1809 7.9273>
   rotation=<rot -0.0007 -89.9999 102.964>
  }
 ]
 dyn_hair_01 [{
   shapeType=Capsule
   length=2.8839
   radius=0.5
   shapeName=dyn_hair_02_capsule
   center=<1.7649 -0.3775 -0.1081>
   rotation=<rot -90 165.9638 179.5154>
  }]
 dyn_hair_02 [{
   shapeType=Capsule
   length=2.8839
   radius=0.5
   shapeName=dyn_hair_02_capsule
   center=<1.7649 -0.3775 0.1505>
   rotation=<rot -88.0484 126.4834 -141.0043>
  }]
 dyn_hair_03 [{
   shapeType=Capsule
   length=6.2425
   radius=0.7067
   shapeName=dyn_hair_03_capsule
   center=<3.3443 -0.3827 0.0351>
   rotation=<rot 75.4831 0 -0.0001>
  }]
 dyn_ponytail_1 [{
   shapeType=Capsule
   length=2.0236
   radius=2.5
   shapeName=dyn_ponytail_1_capsule
   center=<2.7226 -0.4853 0.3708>
   rotation=<rot 40.6402 -89.5507 93.4348>
  }]
 dyn_ponytail_2 [{
   shapeType=Capsule
   length=1.7486
   radius=0.8585
   shapeName=dyn_ponytail_2_capsule
   center=<1.4775 -0.2353 0.1325>
   rotation=<rot -0.3811 80.4278 -87.8381>
  }]
 dyn_ponytail_3 [{
   shapeType=Capsule
   length=2.8562
   radius=2.5123
   shapeName=dyn_ponytail_3_capsule
   center=<2.2882 -1.0171 0.665>
   rotation=<rot 4.1287 -93.6857 -110.3401>
  }]
 dyn_ponytail_4 [
  {
   shapeType=Capsule
   length=1.474
   radius=2.9277
   shapeName=dyn_ponytail_4_capsule
   center=<1.881 0.5846 0.7559>
   rotation=<rot 0.8937 -84.832 -106.0561>
  }
  {
   shapeType=Capsule
   length=1.477
   radius=2.3636
   shapeName=dyn_ponytail_4_capsule
   center=<1.103 0.7673 1.6914>
   rotation=<rot 0.8937 -84.8319 -67.4848>
  }
 ]
 dyn_ponytail_5 [{shapeType=Capsule, length=2.5554, radius=1.1053, shapeName=dyn_ponytail_5_capsule, center=<2.4137 0 -0.3609>, rotation=<rot -90 180 180>}]
 head [
  {shapeType=Capsule, length=2.7428, radius=9.0849, shapeName=head_capsule, center=<5.2443 -0.1147 -0.1497>, rotation=<rot -90 -0.004 0.0037>}
  {
   shapeType=Capsule
   length=6.5457
   radius=5.705
   shapeName=head_capsule
   center=<0.6042 3.1994 -0.1373>
   rotation=<rot -78.4066 179.9995 -180>
  }
 ]
 neck_01 [{shapeType=Capsule, length=1.8448, radius=5.1716, shapeName=neck_01_capsule, rotation=<rot 0 -90.0001 89.0985>}]
 neck_02 [{shapeType=Capsule, length=2.8461, radius=4.3899, shapeName=neck_02_capsule, center=<0 0.3389 0>, rotation=<rot 0 -90 89.648>}]
 spine_03 [{shapeType=Capsule, length=13.5, radius=5.5, shapeName=spine_03_capsule, rotation=<rot 0 -90 0>}]
 spine_04 [{shapeType=Capsule, length=13.5, radius=5.5, shapeName=spine_04_capsule, rotation=<rot 0 -90 0>}]
 spine_05 [
  {shapeType=Capsule, length=10, radius=9, shapeName=spine_05_capsule, center=<0.8119 -2.4091 0>, rotation=<rot 0 -90 0>}
  {
   shapeType=Capsule
   length=31.7788
   radius=15.2168
   shapeName=spine_05_capsule
   center=<13.6975 -19.8812 0.3307>
   rotation=<rot 8.5886 -90.0007 0>
  }
 ]
 upperarm_l [{
   shapeType=Capsule
   length=11.8062
   radius=5.1315
   shapeName=upperarm_l_capsule
   center=<8.5479 -0.2614 -1.0977>
   rotation=<rot -0.0001 -89.9994 87.4073>
  }]
}
constraints [
 {
  bone1Name=spine_04
  bone2Name=spine_03
  swing1LimitDegrees=45
  swing1Motion=Locked
  swing2LimitDegrees=45
  swing2Motion=Locked
  twistLimitDegrees=45
  twistMotion=Locked
 }
 {
  bone1Name=spine_05
  bone2Name=spine_04
  swing1LimitDegrees=45
  swing1Motion=Locked
  swing2LimitDegrees=45
  swing2Motion=Locked
  twistLimitDegrees=45
  twistMotion=Locked
 }
 {
  bone1Name=clavicle_l
  bone2Name=spine_05
  swing1LimitDegrees=45
  swing1Motion=Locked
  swing2LimitDegrees=45
  swing2Motion=Locked
  twistLimitDegrees=45
  twistMotion=Locked
 }
 {
  bone1Name=neck_01
  bone2Name=spine_05
  swing1LimitDegrees=45
  swing1Motion=Locked
  swing2LimitDegrees=45
  swing2Motion=Locked
  twistLimitDegrees=45
  twistMotion=Locked
 }
 {
  bone1Name=neck_02
  bone2Name=neck_01
  swing1LimitDegrees=45
  swing1Motion=Locked
  swing2LimitDegrees=45
  swing2Motion=Locked
  twistLimitDegrees=45
  twistMotion=Locked
 }
 {
  bone1Name=head
  bone2Name=neck_02
  swing1LimitDegrees=45
  swing1Motion=Locked
  swing2LimitDegrees=45
  swing2Motion=Locked
  twistLimitDegrees=45
  twistMotion=Locked
 }
 {
  bone1Name=upperarm_l
  bone2Name=clavicle_l
  swing1LimitDegrees=45
  swing1Motion=Locked
  swing2LimitDegrees=45
  swing2Motion=Locked
  twistLimitDegrees=45
  twistMotion=Locked
 }
 {
  bone1Name=base_ponytail_1
  bone2Name=head
  swing1LimitDegrees=45
  swing1Motion=Locked
  swing2LimitDegrees=45
  swing2Motion=Locked
  twistLimitDegrees=45
  twistMotion=Locked
 }
 {bone1Name=dyn_ponytail_1, bone2Name=base_ponytail_1, swing1LimitDegrees=5, swing2LimitDegrees=5, twistLimitDegrees=0}
 {bone1Name=dyn_ponytail_2, bone2Name=dyn_ponytail_1, swing1LimitDegrees=12, swing2LimitDegrees=12, twistLimitDegrees=0}
 {bone1Name=dyn_ponytail_3, bone2Name=dyn_ponytail_2, swing1LimitDegrees=30, swing2LimitDegrees=25, twistLimitDegrees=0}
 {bone1Name=dyn_ponytail_4, bone2Name=dyn_ponytail_3, swing1LimitDegrees=40, swing2LimitDegrees=24, twistLimitDegrees=10}
 {bone1Name=dyn_ponytail_5, bone2Name=dyn_ponytail_4, swing1LimitDegrees=45, swing2LimitDegrees=30, twistLimitDegrees=30}
 {bone1Name=dyn_hair_03, bone2Name=base_hair_03, swing1LimitDegrees=30, swing2LimitDegrees=20, twistLimitDegrees=2}
 {bone1Name=dyn_hair_02, bone2Name=base_hair_02, swing1LimitDegrees=15, swing2LimitDegrees=15, twistLimitDegrees=2}
 {
  bone1Name=base_hair_03
  bone2Name=head
  swing1LimitDegrees=45
  swing1Motion=Free
  swing2LimitDegrees=45
  swing2Motion=Free
  twistLimitDegrees=45
  twistMotion=Free
 }
 {
  bone1Name=base_hair_02
  bone2Name=head
  swing1LimitDegrees=45
  swing1Motion=Free
  swing2LimitDegrees=45
  swing2Motion=Free
  twistLimitDegrees=45
  twistMotion=Free
 }
 {bone1Name=dyn_hair_01, bone2Name=base_hair_01, swing1LimitDegrees=25, swing2LimitDegrees=15, twistLimitDegrees=0}
 {
  bone1Name=base_hair_01
  bone2Name=head
  swing1LimitDegrees=45
  swing1Motion=Free
  swing2LimitDegrees=45
  swing2Motion=Free
  twistLimitDegrees=45
  twistMotion=Free
 }
]
